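# Physics asset for the TwiceBaked backpack cosmetic, giving the pack's dangling limb bones collision bodies and joint limits.
# Two capsules cover the pack body, and small uniform capsules sit on the dyn_ bones of both legs and the right arm.
# Constraints chain thigh, calf and foot, plus upper arm and forearm, back to the pack, with zero twist limits and asymmetric swing limits.
# The left arm has no body or constraint.
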PhysicsAsset /BRCosmetics/Accessories/FORT_Backpacks/Backpack_M_MED_TwiceBaked/Meshes/M_MED_TwiceBaked_Pack_Physics
bodies {
 dyn_4arm_r [{shapeType=Capsule, length=2, radius=2, shapeName=dyn_4arm_r_capsule, center=<5.5129 0 -1.7518>, rotation=<rot 90 0 0>}]
 dyn_arm_r [{shapeType=Capsule, length=2, radius=2, shapeName=dyn_arm_r_capsule, center=<2 0 0>, rotation=<rot 90 0 0>}]
 dyn_calf_l [{shapeType=Capsule, length=2, radius=2, shapeName=dyn_calf_l_capsule, center=<2 0 0>, rotation=<rot 90 0 0>}]
 dyn_calf_r [{shapeType=Capsule, length=2, radius=2, shapeName=dyn_calf_r_capsule, center=<2 0 0>, rotation=<rot 90 0 0>}]
 dyn_foot_l [{shapeType=Capsule, length=2, radius=2, shapeName=dyn_foot_l_capsule, center=<2 -1.2 0>, rotation=<rot 90 0 -90>}]
 dyn_foot_r [{shapeType=Capsule, length=2, radius=2, shapeName=dyn_foot_r_capsule, center=<2 1.2 0>, rotation=<rot 90 0 -90>}]
 dyn_leg_l [{shapeType=Capsule, length=2, radius=2, shapeName=dyn_leg_l_capsule, center=<2 0 0>, rotation=<rot 90 0 0>}]
 dyn_leg_r [{shapeType=Capsule, length=2, radius=2, shapeName=dyn_leg_r_capsule, center=<2 0 0>, rotation=<rot 90 0 0>}]
 pack [
  {
   shapeType=Capsule
   length=4.5457
   radius=14.1479
   shapeName=pack_capsule
   center=<-1.4514 14.4748 -15.1995>
   rotation=<rot -2.4576 94.0852 -88.4237>
  }
  {
   shapeType=Capsule
   length=30
   radius=7.75
   shapeName=pack_capsule
   center=<-0.9189 7.8432 3.0293>
   rotation=<rot 85.6095 -105.2204 -103.5597>
  }
 ]
}
constraints [
 {bone1Name=dyn_arm_r, bone2Name=pack, swing1LimitDegrees=60, swing2LimitDegrees=8, twistLimitDegrees=0}
 {bone1Name=dyn_4arm_r, bone2Name=dyn_arm_r, swing1LimitDegrees=0, swing2LimitDegrees=20, twistLimitDegrees=0}
 {bone1Name=dyn_leg_l, bone2Name=pack, swing1LimitDegrees=60, swing2LimitDegrees=8, twistLimitDegrees=0}
 {bone1Name=dyn_calf_l, bone2Name=dyn_leg_l, swing1LimitDegrees=30, swing2LimitDegrees=8, twistLimitDegrees=0}
 {bone1Name=dyn_foot_l, bone2Name=dyn_calf_l, swing1LimitDegrees=60, swing2LimitDegrees=8, twistLimitDegrees=0}
 {bone1Name=dyn_leg_r, bone2Name=pack, swing1LimitDegrees=60, swing2LimitDegrees=8, twistLimitDegrees=0}
 {bone1Name=dyn_calf_r, bone2Name=dyn_leg_r, swing1LimitDegrees=30, swing2LimitDegrees=8, twistLimitDegrees=0}
 {bone1Name=dyn_foot_r, bone2Name=dyn_calf_r, swing1LimitDegrees=60, swing2LimitDegrees=8, twistLimitDegrees=0}
]
