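# Physics asset for the left arm of the TwiceBaked backpack cosmetic, defining bodies for dangling or secondary motion on that arm.
# A large capsule covers the pack bone, and two small capsules cover the dyn_arm_l and dyn_4arm_l bones.
# Two constraints chain forearm to arm to pack, with zero twist limits and a tighter swing range at the forearm joint than at the pack joint.
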PhysicsAsset /BRCosmetics/Accessories/FORT_Backpacks/Backpack_M_MED_TwiceBaked/Meshes/M_MED_TwiceBaked_Pack_Physics_larm
bodies {
 dyn_4arm_l [{shapeType=Capsule, length=2, radius=2, shapeName=dyn_4arm_l_capsule, center=<2 0 -1.1037>, rotation=<rot 90 0 0>}]
 dyn_arm_l [{shapeType=Capsule, length=2, radius=2, shapeName=dyn_arm_l_capsule, center=<2 0 0>, rotation=<rot 90 0 0>}]
 pack [{shapeType=Capsule, length=35, radius=8.5293, shapeName=pack_capsule, center=<-0.1993 6.7653 4.3817>, rotation=<rot -19.2573 -90 -88.9745>}]
}
constraints [
 {bone1Name=dyn_arm_l, bone2Name=pack, swing1LimitDegrees=45, swing2LimitDegrees=10, twistLimitDegrees=0}
 {bone1Name=dyn_4arm_l, bone2Name=dyn_arm_l, swing1LimitDegrees=0, swing2LimitDegrees=20, twistLimitDegrees=0}
]
